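; Gameplay ability for the ToyFox boss melee attack, allowed only while the owner carries the melee-target-detected tag.
; It picks a headbutt when frustration or target-proximity duration passes hotfixable thresholds, otherwise alternates left and right swipes, and disables look-at during the attack.
; Montage notifies toggle stagger immunity and trigger targeting, whose hits split into buildings, giant pawns and others for async, tick-throttled damage effects.
; EndAbility cleans up tasks, effects and look-at.
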(event Ability|EventActivateAbility
  (bind _hotfix_delaybeforeattack (Variables|Hotfix|GetHotfix_DelayBeforeAttack))
  (bind _returnvalue (ScalableFloat|GetCachedScalableFloatValue _hotfix_delaybeforeattack))
  (bind _returnvalue_2 (Ability|GetAbilitySystemComponentfromActorInfo))
  ;[0.01374 0.15 0 0.5] ActivateAbility: Perform melee attack and deal damage.
  (|CallParentFunction)
  (Fortnite|FortShippingLog self "ActivateAbility: Invoked")
  (Ability|CommitAbility) ; Applies cooldown.
  (if (> _returnvalue 0)
    (Utilities|FlowControl|Delay _returnvalue) ; Delay before attacking in case we have to provide more openings.
    (else
      _))
  (bind _output_get (Variables|Transient|SetBossActor (Ability|GetAvatarActorfromActorInfo)))
  ;;[0.01208 0.06546 0.15 0.5] Disallow looking at target when melee attacking.
  (Default|SetAllowLookatTarget(Message) _output_get false)
  (bind (_animmontage _targetingpreset _gameplayeffectlevel _gameplayeffectstoapply) (CallFunction|GetPreferredMeleeAttackProperties))
  (Variables|Transient|SetAttackGameplayEffectLevel _gameplayeffectlevel)
  (CallFunction|ApplyGameplayEffectstoOwner _gameplayeffectstoapply)
  (bind (_asynctaskproxy _notifytag) (Ability|Tasks|PlayMontageAndWaitNotify "MeleeMontageTask" _animmontage (ScalableFloat|GetCachedScalableFloatValue (Variables|Hotfix|GetHotfix_MeleeAnimPlayRate)))
    (:then
      (Variables|Transient|SetAsyncTask_MeleeMontage _asynctaskproxy))
    (:OnCompleted
      _)
    (:OnInterrupted
      _)
    (:OnCancelled
      _)
    (:OnNotifyBegin
      (if (== _notifytag "(TagName=\"TidalCrane.ToyFox.AnimNotify.Ability.Melee.Impact\")")
        ;;[0.006845 0.001379 0.15 0.5] Perform targeting and apply melee damage to the various hit actors.
        (bind (_asynctaskproxy_2 _targetdata) (Ability|Tasks|PerformTargetingAbilityTask "TargetingTask" _targetingpreset false false true)
          (:then
            (Variables|Transient|SetAsyncTask_Targeting _asynctaskproxy_2))
          (:Targeted
            _)
          (:Cancelled
            _)) ; Target for builids
        (bind (_buildingactors _nonbuildingactors _giantpawns) (CallFunction|ParseTargetData _targetdata))
        (Utilities|FlowControl|Sequence
          (:then_0
            (bind _asynctaskproxy_3 (Ability|Tasks|ApplyGameplayEffectToHitResultsAsync :OriginatingAbilitySystemComponent _returnvalue_2 :HitResults _buildingactors :GameplayEffectClass (Variables|Settings|GetDefaultMeleeDamageGameplayEffectClass) :GameplayEffectLevel (Variables|Transient|GetAttackGameplayEffectLevel) :Stacks 1 :MaxTargetsToProcessPerTickCVar (Variables|Hotfix|GetHotfix_MaxTargetsProcessPerTick)))
            (Utilities|Array|Add (Variables|Transient|GetTargetingDamageApplicationAsyncActions) _asynctaskproxy_3))
          (:then_1
            (bind _asynctaskproxy_4 (Ability|Tasks|ApplyGameplayEffectToHitResultsAsync :OriginatingAbilitySystemComponent _returnvalue_2 :HitResults _nonbuildingactors :GameplayEffectClass (Variables|Settings|GetDefaultMeleeDamageGameplayEffectClass) :GameplayEffectLevel (Variables|Transient|GetAttackGameplayEffectLevel) :Stacks 1 :MaxTargetsToProcessPerTickCVar (Variables|Hotfix|GetHotfix_MaxTargetsProcessPerTick)))
            (Utilities|Array|Add (Variables|Transient|GetTargetingDamageApplicationAsyncActions) _asynctaskproxy_4))
          (:then_2
            (bind _asynctaskproxy_5 (Ability|Tasks|ApplyGameplayEffectToHitResultsAsync :OriginatingAbilitySystemComponent _returnvalue_2 :HitResults _giantpawns :GameplayEffectClass (Variables|Settings|GetGiantMeleeDamageGameplayEffectClass) :GameplayEffectLevel (Variables|Transient|GetAttackGameplayEffectLevel) :Stacks 1 :MaxTargetsToProcessPerTickCVar (Variables|Hotfix|GetHotfix_MaxTargetsProcessPerTick)))
            (Utilities|Array|Add (Variables|Transient|GetTargetingDamageApplicationAsyncActions) _asynctaskproxy_5)))
        (elif (== _notifytag "(TagName=\"TidalCrane.ToyFox.AnimNotify.Ability.Melee.StaggerImmune\")")
          ;;[0.01633 0.01865 0.15 0.5] Handle stagger immunity during the attacks.
          ;; During portions of the attack animation, the bug cannot be staggered.
          (Ability|RemoveGameplayEffectFromOwnerWithHandle (Variables|Transient|GetStaggerImmunityEffectHandle))
          (bind _returnvalue_3 (Ability|ApplyGameplayEffectToOwner (Variables|Settings|GetStaggerImmunityGameplayEffectClass)))
          (Variables|Transient|SetStaggerImmunityEffectHandle _returnvalue_3))))
    (:OnNotifyEnd
      (if (== _notifytag "(TagName=\"TidalCrane.ToyFox.AnimNotify.Ability.Melee.StaggerImmune\")")
        (Ability|RemoveGameplayEffectFromOwnerWithHandle (Variables|Transient|GetStaggerImmunityEffectHandle)))))
  (Ability|EndAbility))

(event Ability|EventOnEndAbility
  (params (bWasCancelled :type Boolean))
  (bind _targetingdamageapplicationasyncactions (Variables|Transient|GetTargetingDamageApplicationAsyncActions))
  ;[0.08333 0.05408 0 0.5] EndAbility: Clean up ability.
  (Utilities|FlowControl|Sequence
    (:then_0
      (GameplayTasks|EndTask (Variables|Transient|GetAsyncTask_MeleeMontage))
      (Ability|RemoveGameplayEffectFromOwnerWithHandle (Variables|Transient|GetStaggerImmunityEffectHandle))
      (Ability|Async|EndAction _targetingdamageapplicationasyncactions)
      (Utilities|Array|Clear _targetingdamageapplicationasyncactions)
      (bind _bossactor (Variables|Transient|GetBossActor)
        (:then
          (Default|SetAllowLookatTarget(Message) _bossactor true))
        (:else)))
    (:then_1
      (|CallParentFunction bWasCancelled))))

(fn ApplyGameplayEffectsToOwner
  (params (GameplayEffectClass :type Array%20of%20Gameplay%20Effect%20Class%20References))
  (for _array_element GameplayEffectClass
    (Ability|ApplyGameplayEffectToOwner _array_element)))

(fn GetPreferredMeleeAttackProperties
  ;[0 0.15 0.03215 0.5] Returns the melee attack properties to use.
  (Utilities|FlowControl|Sequence
    (:then_0
      (bind _asbga_tidal_crane_boss_toy_fox (Utilities|Casting|CastToBGA_TidalCrane_Boss_ToyFox (Ability|GetAvatarActorfromActorInfo)))
      (if (or (> (Frustration|GetFrustration _asbga_tidal_crane_boss_toy_fox) (ScalableFloat|GetCachedScalableFloatValue (Variables|Hotfix|GetHotfix_HeadbuttUseFrustrationThreshold))) (> (TargetProximity|GetTargetProximityDuration _asbga_tidal_crane_boss_toy_fox) (ScalableFloat|GetCachedScalableFloatValue (Variables|Hotfix|GetHotfix_HeadbuttUseProximityDuration))))
        (return (Variables|Settings|GetHeadbuttAttackMontage) (Variables|Settings|GetHeadbuttTargetingPreset) 1 (Utilities|Array|MakeArray (Variables|Settings|GetHeadbuttRecentlyUsedGameplayEffectClass)))
        (else
          (bind _output_get (Variables|Transient|SetbRightHandAttack (not (Variables|Transient|GetbRightHandAttack))))
          (if _output_get
            (return (Variables|Settings|GetLeftMeleeAttackMontage) (Variables|Settings|GetMeleeAttackTargetingPreset) 0 0)
            (else
              (return (Variables|Settings|GetRightMeleeAttackMontage) (Variables|Settings|GetMeleeAttackTargetingPreset) 0 0))))))
    (:then_1
      (return "/TidalCrane_ToyFox_Boss/Assets/Animations/TidalCrane_ToyFox_Boss_MeleeAttackR_Montage.TidalCrane_ToyFox_Boss_MeleeAttackR_Montage" 0 0 0))))

(fn CanActivateAbility
  (params (ActorInfo :type Gameplay%20Ability%20Actor%20Info%20Structure) (Handle :type Gameplay%20Ability%20Spec%20Handle%20Structure))
  (bind _owneractor (Utilities|Struct|BreakGameplayAbilityActorInfo ActorInfo))
  (Utilities|FlowControl|Sequence
    (:then_0
      (Utilities|IsValid _owneractor
        (:Is%20Valid
          (bind _returnvalue (GameplayTags|HasMatchingGameplayTag(Message) _owneractor "(TagName=\"TidalCrane.ToyFox.Boss.Melee.Target.Detected\")"))
          (return 0 _returnvalue))))
    (:then_1
      (return 0 false))))

(fn ParseTargetData
  (params (TargetDataHandle :type Gameplay%20Ability%20Target%20Data%20Handle%20Structure))
  (for _index (range (Ability|TargetData|GetDataCountfromTargetData TargetDataHandle))
    (bind _output_get (LocalVariables|SetHitResult (Ability|TargetData|GetHitResultfromTargetData TargetDataHandle _index)))
    (Utilities|Casting|CastToBuildingActor (.HitActor (Collision|BreakHitResult _output_get))
      (:then
        (Utilities|Array|Add (LocalVariables|GetRetBuildingActors) (LocalVariables|GetHitResult)))
      (:CastFailed
        (Utilities|Casting|CastToFortGiantPlayerPawn (.HitActor (Collision|BreakHitResult _output_get))
          (:then
            (Utilities|Array|Add (LocalVariables|GetRetGiantPawns) (LocalVariables|GetHitResult)))
          (:CastFailed
            (Utilities|IsValid (.HitActor (Collision|BreakHitResult _output_get))
              (:Is%20Valid
                (Utilities|Array|Add (LocalVariables|GetRetNonBuildingActors) (LocalVariables|GetHitResult)))))))))
  (return (LocalVariables|GetRetBuildingActors) (LocalVariables|GetRetNonBuildingActors) (LocalVariables|GetRetGiantPawns)))
