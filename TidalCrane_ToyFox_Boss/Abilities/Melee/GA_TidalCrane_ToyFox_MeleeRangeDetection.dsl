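; Gameplay ability for the ToyFox boss that tracks whether a valid melee target is in range while active.
; A looping timer, at a hotfixable interval, runs a targeting-preset task while the avatar's StateTree is running.
; Results apply a target-detected gameplay effect to the owner, or remove it by granted tag.
; A delta-gated toggle starts detection on activation; ending clears the timer, effect and targeting task.
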(event Ability|EventActivateAbility
  ;[0.02145 0.1255 0.15 0.5] ActivateAbility: Start melee range detection.
  (|CallParentFunction)
  (Fortnite|FortShippingLog self "ActivateAbility: Invoked")
  (CallFunction|SetEnableMeleeRangeDetection true)
  (bind _returnvalue (Ability|GetAvatarActorfromActorInfo))
  (Utilities|IsValid _returnvalue
    (:Is%20Valid
      (Variables|Transient|SetStateTreeComponent (Actor|GetComponentbyClass _returnvalue "/Script/GameplayStateTreeModule.StateTreeComponent")))))

(event Custom|SetEnableMeleeRangeDetection
  (params (bEnable :type Boolean))
  ;[0.01057 0.15 0.08862 0.5] SetEnableMeleeRangeDetection: Start or stop deteching melee range targets.
  (bind _updated_variable (Utilities|FlowControl|DeltaGateWildcard (Variables|Transient|GetbMeleeRangeDetectionEnabled) bEnable)
    (:Success
      (Fortnite|FortShippingLog self (Utilities|String|Append "SetEnableMeleeRangeDetection: Set as " (Utilities|String|ToString(Boolean) _updated_variable)))
      (if _updated_variable
        (bind _returnvalue (Utilities|Time|SetTimerbyEvent (EventDispatchers|CreateEvent :!event Custom|ExecuteMeleeRangeTrace) (ScalableFloat|GetCachedScalableFloatValue (Variables|Hotfix|GetHotfix_MeleeRangeTraceFrequency)) true true))
        (Variables|Transient|SetMeleeRangeTraceTimerHandle _returnvalue)
        (CallFunction|ExecuteMeleeRangeTrace)
        (else
          (Utilities|Time|ClearandInvalidateTimerbyHandle (Variables|Transient|GetMeleeRangeTraceTimerHandle))
          (Ability|RemoveGameplayEffectFromOwnerWithGrantedTags "(GameplayTags=((TagName=\"TidalCrane.ToyFox.Boss.Melee.Target.Detected\")))")
          (bind _targetingtask (Variables|Transient|GetTargetingTask)
            (:then
              (GameplayTasks|EndTask _targetingtask))
            (:else)))))))

(event Custom|ExecuteMeleeRangeTrace
  ;[0.15 0.03178 0.1388 0.5] ExecuteMeleeRangeTrace: Run trace to figure out if there's a valid melee target, like the giant pawn.
  (bind _statetreecomponent (Variables|Transient|GetStateTreeComponent)
    (:then
      (if (Utilities|Enum|Equal(Enum) (Gameplay|StateTree|GetStateTreeRunStatus _statetreecomponent :!class /Script/GameplayStateTreeModule.StateTreeComponent) "Running")
        (bind (_asynctaskproxy _targetdata) (Ability|Tasks|PerformTargetingAbilityTask "MeleeRangeTraceTask" (Variables|Settings|GetTargetingPresetMeleeRangeDetection) false false true)
          (:then
            (Variables|Transient|SetTargetingTask _asynctaskproxy))
          (:Targeted
            _)
          (:Cancelled
            _))
        (if (Variables|Transient|GetbMeleeRangeDetectionEnabled)
          (if (Ability|TargetData|TargetDataHasActor _targetdata)
            (Ability|ApplyGameplayEffectToOwner (Variables|Settings|GetMeleeRangeTargetDetectedGameplayEffectClass))
            (else
              (Ability|RemoveGameplayEffectFromOwnerWithGrantedTags "(GameplayTags=((TagName=\"TidalCrane.ToyFox.Boss.Melee.Target.Detected\")))"))))))
    (:else)))

(event Ability|EventOnEndAbility
  (params (bWasCancelled :type Boolean))
  ;[0.15 0.08593 0 0.5] OnEndAbility: Clean up.
  (Utilities|FlowControl|Sequence
    (:then_0
      (Fortnite|FortShippingLog self "OnEndAbility: Invoked")
      (CallFunction|SetEnableMeleeRangeDetection false))
    (:then_1
      (|CallParentFunction bWasCancelled))))

(fn CanActivateAbility
  (params (ActorInfo :type Gameplay%20Ability%20Actor%20Info%20Structure) (Handle :type Gameplay%20Ability%20Spec%20Handle%20Structure))
  (return 0 true))
