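; Targeting task for the TidalCrane ToyFox boss's melee attack that makes the targeting source offset relative to the source actor's orientation.
; Its GetSourceOffset override reads the parent's offset and applies X, Y and Z along the actor's forward, right and up vectors.
; If the targeting subsystem or source actor is invalid, it returns the parent's offset unchanged.
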(fn GetSourceOffset
  (params (TargetingHandle :type Targeting%20Request%20Handle%20Structure%20%28by%20ref%29))
  (bind _returnvalue (TargetingSystem|TargetingTypes|GetTargetingSourceContext (Targeting|GetTargetingSubsystem TargetingHandle) TargetingHandle))
  (bind _sourceactor (Utilities|Struct|BreakTargetingSourceContext _returnvalue))
  (bind _returnvalue_2 (Targeting|GetTargetingSubsystem TargetingHandle))
  (Utilities|FlowControl|Sequence
    (:then_0
      (Utilities|IsValid _returnvalue_2
        (:Is%20Valid
          (Utilities|IsValid _sourceactor
            (:Is%20Valid
              (bind _returnvalue_3 (|CallParentFunction TargetingHandle))
              (bind (_returnvalue_x _returnvalue_y _returnvalue_z) (Math|Vector|BreakVector _returnvalue_3))
              (return (Utilities|Operators|Add (* _returnvalue_x (Transformation|GetActorForwardVector _sourceactor)) (* _returnvalue_y (Transformation|GetActorRightVector _sourceactor)) (* _returnvalue_z (Transformation|GetActorUpVector _sourceactor)))))))))
    (:then_1
      (bind _returnvalue_4 (|CallParentFunction TargetingHandle))
      (return _returnvalue_4))))
